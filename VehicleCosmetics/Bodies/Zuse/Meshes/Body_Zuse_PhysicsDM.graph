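# Physics asset for the Zuse vehicle body cosmetic, giving its skeletal mesh a simplified collision shape.
# It holds a single body on the root bone, made of one box offset forward and upward from the bone origin.
# It has no constraints, so the whole vehicle body is one rigid shape with no articulated parts.
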PhysicsAsset /VehicleCosmetics/Bodies/Zuse/Meshes/Body_Zuse_PhysicsDM
bodies {root [{shapeType=Box, extentX=444, extentY=216, extentZ=116, shapeName=root_box, center=<29.9011 0 95.0189>}]}
constraints []
