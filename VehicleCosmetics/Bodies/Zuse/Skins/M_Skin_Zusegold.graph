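# Vehicle body material for the Zusegold skin of the Zuse body, a metallic gold finish for skeletal meshes, with mesh deformer usage enabled.
# Color, Metal and Roughness parameters plus a BodyNormalLOD normal map, blank by default, build a material attributes set.
# That set passes through the shared MF_BaseVehicle_Damage function, then MF_VehicleDeRez_Transition, layering vehicle damage and de-rez transition handling onto the base look.
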Material /VehicleCosmetics/Bodies/Zuse/Skins/M_Skin_Zusegold
graph {
 properties {
  bAutomaticallySetUsageInEditor=false
  bUseMaterialAttributes=true
  bUsedWithMeshDeformer=true
  bUsedWithSkeletalMesh=true
  usedShadingModels=Default Lit
 }
 nodes [
  n1 = VectorParameter() {parameterName=Color, defaultValue=<rgba 0.5922 0.4078 0.1608 1>}
  n2 = ScalarParameter() {defaultValue=1, parameterName=Metal}
  n3 = ScalarParameter() {defaultValue=0.22, parameterName=Roughness}
  n4 = TextureSampleParameter2D() {parameterName=BodyNormalLOD, samplerType=SAMPLERTYPE_Normal, texture=/VehicleCosmetics/Textures/T_BlankNormal_B}
  MaterialFunctionCall(In=MaterialFunctionCall(Material=MakeMaterialAttributes!(
     BaseColor=@n1.RGB
     Metallic=@n2
     Specular
     Roughness=@n3
     Anisotropy
     EmissiveColor
     Opacity
     OpacityMask
     Normal=@n4.RGB
     Tangent
     WorldPositionOffset
     SubsurfaceColor
     ClearCoat
     ClearCoatRoughness
     AmbientOcclusion
     Refraction
     CustomizedUVs_0
     CustomizedUVs_1
     CustomizedUVs_2
     CustomizedUVs_3
     CustomizedUVs_4
     CustomizedUVs_5
     CustomizedUVs_6
     CustomizedUVs_7
     PixelDepthOffset
     ShadingModel
     Displacement
    )) {materialFunction=&MF_BaseVehicle_Damage}.Result) {materialFunction=&MF_VehicleDeRez_Transition}
 ]
}
